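; Speech bubble widget component for the SouthBulkAI NPC that adds a vertical offset to the bubble widget.
; It overrides UpdatedWidgetOffset, runs the parent logic, then adds the YOffset scalable float to the user widget's render translation Y.
; The render transform's X translation, scale, shear and angle are carried through unchanged.
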(event EventUpdatedWidgetOffset
  (bind (_rendertransform_translation _rendertransform_scale _rendertransform_shear _rendertransform_angle) (Utilities|Struct|BreakWidgetTransform (Variables|/Script/UMG.Widget|GetRenderTransform_Translation_X (UserInterface|GetUserWidgetObject))))
  (bind (_rendertransform_translation_x _rendertransform_translation_y) (Math|Vector2D|BreakVector2D _rendertransform_translation))
  (|CallParentFunction)
  (Variables|/Script/UMG.Widget|SetRenderTransform_Translation_X (Utilities|Struct|MakeWidgetTransform (Math|Vector2D|MakeVector2D _rendertransform_translation_x (+ _rendertransform_translation_y (ScalableFloat|GetValueatLevel (Variables|Default|GetYOffset)))) _rendertransform_scale _rendertransform_shear _rendertransform_angle) (UserInterface|GetUserWidgetObject)))
